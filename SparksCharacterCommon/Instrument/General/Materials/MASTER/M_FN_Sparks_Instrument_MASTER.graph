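# Master material for Sparks instruments, built on material attributes with the shading model chosen by the material expression and usable on skeletal, clothing, morph, deformer and Nanite meshes.
# A pre-FX base function outputs attributes plus AO, Gmap, pre-skinned position and other masks.
# An instrument FX function modifies those attributes, then an instrument post-FX function also takes distance blend and Fresnel.
# A cinematic character post-FX function finishes the chain.
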Material /SparksCharacterCommon/Instrument/General/Materials/MASTER/M_FN_Sparks_Instrument_MASTER
graph {
 properties {
  bAutomaticallySetUsageInEditor=false
  bUseMaterialAttributes=true
  bUsedWithClothing=true
  bUsedWithMeshDeformer=true
  bUsedWithMorphTargets=true
  bUsedWithNanite=true
  bUsedWithSkeletalMesh=true
  pixelDepthOffsetMode=PDOM_Legacy
  refractionMethod=RM_IndexOfRefraction
  shadingModel=MSM_FromMaterialExpression
  subsurfaceProfile=&SS_HeroSkin_02
  usedShadingModels=Default Lit | Subsurface | Clear Coat | Subsurface Profile | Cloth | Eye
 }
 nodes [
  n1 = MaterialFunctionCall!(TexCoords) {materialFunction=/SparksCharacterCommon/Instrument/General/Materials/MF/MF_Base_Instrument_PreFX}
  MaterialFunctionCall(In=MaterialFunctionCall!(
    AO=@n1.AO
    Distance Blend=@n1.Distance Blend
    Fresnel=@n1.Fresnel
    Gmap=@n1.Gmap
    Pre Skinned Local Position=@n1.Pre skinned Local Position
    Material=MaterialFunctionCall!(PreskinnedPosition=@n1.Pre skinned Local Position, Gmap=@n1.Gmap, MaterialAttributes=@n1.Result, AO=@n1.AO) {materialFunction=/SparksCharacterCommon/Instrument/General/Materials/MF/MF_InstrumentFX}.Result
   ) {materialFunction=/SparksCharacterCommon/Instrument/General/Materials/MF/MF_Instruments_PostFX}.Result) {materialFunction=&CINE_MF_CharPostFX}
 ]
}
